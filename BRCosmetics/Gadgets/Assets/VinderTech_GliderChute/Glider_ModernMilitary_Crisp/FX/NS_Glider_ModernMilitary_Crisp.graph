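# Niagara effect for the ModernMilitary Crisp glider, adding two spinning propellers and two ribbon trails attached to the glider skeletal mesh.
# Prop places mesh particles on bones, with an interpolated motion value raising spin rate and crossfading the solid prop into blur and glow.
# TrailsBig spawns anim-trail ribbons whose velocity and opacity follow interpolated values.
# Front-end modifiers substitute menu-specific behaviour, and time-of-day functions adjust blur opacity and trail tint.
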NiagaraSystem /BRCosmetics/Gadgets/Assets/VinderTech_GliderChute/Glider_ModernMilitary_Crisp/FX/NS_Glider_ModernMilitary_Crisp
dependencies {
 usedDataInterfaces [&NiagaraDataInterfaceArrayFloat3, &NiagaraDataInterfaceCurve, &NiagaraDataInterfaceArrayFloat]
 usedDynamicInputs [&SetBoolByFloatComparison]
 usedModules [
  &self:SystemSpawnScript.SystemScriptSource.SystemScriptGraph.NiagaraNodeAssignment_3.SetVariables_#1
  &SystemState
  &FN_Integrator
  &Timeline
  &InterpolateOverTime
  &self:SystemSpawnScript.SystemScriptSource.SystemScriptGraph.NiagaraNodeAssignment_0.SetVariables_#2
 ]
}
emitters {
 Prop {
  emitterUpdateScript [
   &EmitterState(Life Cycle Mode=::NewEnumerator1, Loop Behavior=::NewEnumerator1, Loop Delay=0.935, Loop Duration Mode=::NewEnumerator1, UseLoopDelay=true)
   &SpawnBurst_Instantaneous(Spawn Count=2, Use Spawn Probability=false, UseLoopCountLimit=false)
  ]
  particleSpawnScript [
   &InitializeParticle(Color Mode=::NewEnumerator0, Lifetime=5, Mesh Scale Mode=::NewEnumerator3, UsePositionOffset=false, Write Material Random=true)
   /SetVariables(Particles.MotionBlurScale:Vector3f=<0.5 0.5 0.1>, Particles.MotionGlowScale:Vector3f=<0.34 0.34 0.1>) {isSetParameters=true}
   /SetVariables(Particles.MotionBlurDynamic:Vector4f=&MakeVector4(X=-4, Y=6, Z=0, W=0)) {isSetParameters=true}
   &PartitionParticles()
   &InitialMeshOrientation(Mesh Orientation Mode=::NewEnumerator2, Use Rotation Vector=false)
  ]
  particleUpdateScript [
   &ParticleState(
    Kill Particles When Lifetime Has Elapsed=false
    Let Infinitely Lived Particles Die When Emitter Deactivates=true
    Loop Particles Lifetime=true
   )
   &ScaleMeshSize(Scale Factor=&Multiply_VectorByFloat(
     Vector=<1 1 1>
     Float=&FloatFromCurve(
      FloatCurve=DataInterface() {propertyValues {Curve {keys [{tangentMode=RCTM_User, time=0, value=0}, {tangentMode=RCTM_User, time=1.5, value=1}]}}}
      CurveIndex=$Engine.System.Age
     )
    ))
   &SkeletalMeshLocation(
    Bone / Socket Index=$Particles.PartitionParticles.Partition
    Bone Sampling Mode=::NewEnumerator4
    MaskSpawnGroup=false
    Orientation Sampling=::NewEnumerator1
    Skeletal Mesh=$System.Glider Skeletal Mesh
   )
   &UpdateMeshOrientation(
    Rotation Coordinate Space=::NewEnumerator0
    Rotation Rate=&FN_FrontEndFloatModifier(
     Gameplay=&Lerp_Float(A=3, B=10, Alpha=$System.InterpolateOverTime.InterpolateOverTime_Float.Moving Average)
     Front End=10
    )
    Rotation Vector=<0 0 1>
   )
   /SetVariables(
    Particles.BlurPropColor:LinearColor=&MakeLinearColorFromVectorAndFloat(
     "Vector (RGB)"=<1 1 1>
     "Float (Alpha)"=&Multiply_Float(
      A=&FNTimeOfDayScalar(Night Scale=0.25, Morning Scale=0.5, Evening Scale=0.5)
      B=&FN_FrontEndFloatModifier(
       Gameplay=&FloatFromCurve(
        FloatCurve=DataInterface() {propertyValues {Curve {keys [{tangentMode=RCTM_User, time=0, value=0}, {tangentMode=RCTM_User, time=0.5, value=1}]}}}
        CurveIndex=$System.InterpolateOverTime.InterpolateOverTime_Float.Moving Average
       )
       Front End=1
      )
     )
    )
    Particles.BlurPropGlowColor:LinearColor=&MakeLinearColorFromVectorAndFloat(
     "Vector (RGB)"=<1 1 1>
     "Float (Alpha)"=&Multiply_Float(
      A=0.05
      B=&FN_FrontEndFloatModifier(
       Gameplay=&FloatFromCurve(
        FloatCurve=DataInterface() {propertyValues {Curve {keys [{tangentMode=RCTM_User, time=0, value=0}, {tangentMode=RCTM_User, time=0.5, value=1}]}}}
        CurveIndex=$System.InterpolateOverTime.InterpolateOverTime_Float.Moving Average
       )
       Front End=1
      )
     )
    )
    Particles.MainPropColor:LinearColor=&MakeLinearColorFromVectorAndFloat(
     "Vector (RGB)"=<1 1 1>
     "Float (Alpha)"=&FN_FrontEndFloatModifier(
      Gameplay=&FloatFromCurve(
       FloatCurve=DataInterface() {propertyValues {Curve {keys [{tangentMode=RCTM_User, time=0, value=1}, {tangentMode=RCTM_User, time=0.5, value=0}]}}}
       CurveIndex=$System.InterpolateOverTime.InterpolateOverTime_Float.Moving Average
      )
      Front End=0
     )
    )
   ) {isSetParameters=true}
   /SetVariables(
    Particles.MotionBlurScale:Vector3f=&Multiply_VectorByFloat(
     Vector=$Particles.Initial.MotionBlurScale
     Float=&FN_FrontEndFloatModifier(Gameplay=1, Front End=$System.SystemFadeIn)
    )
    Particles.MotionGlowScale:Vector3f=&Multiply_VectorByFloat(
     Vector=$Particles.Initial.MotionGlowScale
     Float=&FN_FrontEndFloatModifier(Gameplay=1, Front End=$System.SystemFadeIn)
    )
   ) {isSetParameters=true}
  ]
  rendererClasses [&NiagaraMeshRendererProperties]
  renderers [
   {rendererClass=&NiagaraMeshRendererProperties, rendererIndex=0}
   {rendererClass=&NiagaraMeshRendererProperties, rendererIndex=1}
   {rendererClass=&NiagaraMeshRendererProperties, rendererIndex=2}
  ]
 }
 TrailsBig {
  emitterSpawnScript [&FN_RibbonIDOffsetUtility()]
  emitterUpdateScript [
   &EmitterState()
   &FN_SpawnAnimTrail(Number of Anim Trails=2)
  ]
  particleSpawnScript [
   &FN_InitializeAnimTrailBoneIndecies()
   &FN_InitializeAnimTrailBoneIndecies(AnimTrail Selection=1, First Bone or Socket Index=2, Second Bone or Socket Index=3) {module=FN_InitializeAnimTrailBoneIndecies001}
   /SetVariables(Particles.AlphaFadeIn:NiagaraFloat=$System.SystemFadeIn) {isSetParameters=true}
   &FN_InitializeAnimTrail(
    Color=&MakeLinearColorFromVectorAndFloat("Vector (RGB)"=<1 1 1>, "Float (Alpha)"=&Multiply_Float(A=0.25, B=$Particles.AlphaFadeIn))
    Lifetime=&MakeCustomFloatFromBool(Bool=$Particles.Locked, True Float=2, False Float=0.5)
    Position Offset=&FN_FrontEndVector3Modifier(
     Gameplay=&SelectVectorFromArray(Direct Array Index=$Particles.SpawnGroup, Vector Selection Array=DataInterface() {propertyValues {FloatData [<0 1 -2>, <0 -1 -2>]}})
     Front End=&SelectVectorFromArray(Direct Array Index=$Particles.SpawnGroup, Vector Selection Array=DataInterface() {propertyValues {FloatData [<0 -1 1>, <0 1 1>]}})
    )
    SkeletalMesh=$System.Glider Skeletal Mesh
   )
   &AddVelocity(Velocity=&Multiply_VectorByFloat(
     Vector=&MakeCustomVectorFromBool(
      Bool=$User.bIsFrontEnd
      True Vector=<1000 0 1>
      False Vector=&MakeVector(
       X=&Lerp_Float(A=100, B=400, Alpha=$System.InterpolateOverTime.InterpolateOverTime_Float.Moving Average)
       Y=&Lerp_Float(
        A=0
        B=&Lerp_Float(
         A=0
         B=10000
         Alpha=&FloatFromCurve(FloatCurve=DataInterface() {propertyValues {Curve {keys [{interpMode=RCIM_Cubic, tangentMode=RCTM_User, time=0, value=0}, {interpMode=RCIM_Cubic, tangentMode=RCTM_User, time=0.5, value=1.05}]}}})
        )
        Alpha=$System.InterpolateOverTime001.InterpolateOverTime_Float.Moving Average
       )
       Z=0
      )
     )
     Float=1
    ))
   &DynamicMaterialParameters(
    Index 0 Param 1=&SelectFloatFromArray(Direct Array Index=$Particles.SpawnGroup, Float Selection Array=DataInterface() {propertyValues {FloatData [0.2,0.35]}})
    Index 0 Param 2=&SelectFloatFromArray(Direct Array Index=$Particles.SpawnGroup, Float Selection Array=DataInterface() {propertyValues {FloatData [0.6,0.25]}})
    Module.Index0Param0Write=true
    Module.Index0Param1Write=true
    Module.Index0Param2Write=false
    Module.Index0Param3Write=false
   )
  ]
  particleUpdateScript [
   &SolveForcesAndVelocity(Clamp Velocity=false, Limit Acceleration=false)
   &FN_UpdateAnimTrail(
    Locked Particle Color:LinearColor=&MultiplyLinearColor(
     Linear Color=$Particles.Initial.Color
     Scale Factor=&MakeFloat4FromVectorAndFloat("Vector (RGB)"=<1 1 1>, "Float (Alpha)"=1)
    )
    Locked Particle RibbonWidth=1
    SkeletalMesh=$System.Glider Skeletal Mesh
   ) {module=FN_UpdateAnimTrail001}
   /SetVariables(Particles.RibbonU0Override=$Particles.NormalizedAge) {isSetParameters=true}
   &ScaleColor(
    Scale Alpha=&Multiply_Float(
     A=&FNTimeOfDayScalar(Night Scale=0.5, Morning Scale=1, Evening Scale=1)
     B=&FN_FrontEndFloatModifier(
      Gameplay=&Lerp_Float(A=0, B=0.75, Alpha=$System.InterpolateOverTime.InterpolateOverTime_Float.Moving Average)
      Front End=&Lerp_Float(
       A=0
       B=0.4
       Alpha=&FloatFromCurve(
        FloatCurve=DataInterface() {propertyValues {Curve {keys [{tangentMode=RCTM_User, time=1, value=0}, {tangentMode=RCTM_User, time=2, value=1}]}}}
        CurveIndex=$Engine.System.Age
       )
      )
     )
    )
    Scale RGB=&MakeVectorFromLinearColorRGB(Color=&FNTimeOfDayColor(
      Night Color=<rgba 0.75 0.8179 1 1>
      Morning Color=<rgba 1 0.8086 0.75 1>
      Day Color=<rgba 1 0.9047 0.7231 1>
      Evening Color=<rgba 0.9647 0.7333 0.7235 1>
     ))
    ScaleA=true
    ScaleRGB=true
   )
  ]
  rendererClasses [&NiagaraRibbonRendererProperties]
  renderers [{rendererClass=&NiagaraRibbonRendererProperties}]
 }
}
summary {
 systemName=NS_Glider_ModernMilitary_Crisp
 emitters [
  {emitterName=TrailsBig, rendererClasses [&NiagaraRibbonRendererProperties]}
  {emitterName=Prop, rendererClasses [&NiagaraMeshRendererProperties]}
 ]
 userVariables [
  User.bIsFrontEnd {type=NiagaraBool}
  User.bIsFrontEndPreview {type=NiagaraBool}
  User.bIsFullyDeployed {type=NiagaraBool}
  User.DelayDeployTime {type=NiagaraFloat, defaultValue=0, description=""}
  User.ForwardDot {type=NiagaraFloat, defaultValue=0, description=""}
  User.RibbonIDOffset {type=NiagaraInt32}
  User.RightDot {type=NiagaraFloat, defaultValue=0, description=""}
  User.SkeletalMesh {
   type=NiagaraDataInterfaceSkeletalMesh
   defaultValue {
    struct=NiagaraExt_VariableValue_DataInterface
    value {dataInterface=&self:NiagaraDataInterfaceSkeletalMesh_1}
   }
  }
 ]
}
